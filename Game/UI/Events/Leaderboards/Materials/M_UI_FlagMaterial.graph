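# Translucent UI material for event leaderboards that draws a flag or crest texture cropped to a soft-edged circle, defaulting to a Uruguay crest.
# A radial gradient through SmoothStep forms outer and inner masks with adjustable size and softness.
# The texture is scaled about its center and shifted diagonally by Offset's red channel, its alpha multiplied by the outer mask.
# The IsEmpty static switch instead shows a solid-color ring between the two masks.
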Material /Game/UI/Events/Leaderboards/Materials/M_UI_FlagMaterial
graph {
 properties {
  bUsedWithStaticMesh=false
  blendMode=BLEND_Translucent
  materialDomain=MD_UI
  pixelDepthOffsetMode=PDOM_Legacy
  usedShadingModels=Default Lit
 }
 nodes [
  n1 = MaterialFunctionCall() {materialFunction=&RadialGradientExponential}
  n2 = OneMinus(None=ScalarParameter() {defaultValue=0.5, parameterName=InnerRingSize})
  n3 = MaterialFunctionCall(Alpha=@n1.RadialGradientExponential, Min=@n2, Max=Add(A=@n2, B=ScalarParameter() {defaultValue=0.02, parameterName=EdgeSoftnessInner})) {materialFunction=&SmoothStep}
  n4 = OneMinus(None=ScalarParameter() {defaultValue=1, parameterName=OuterRingSize})
  n5 = MaterialFunctionCall(Alpha=@n1.RadialGradientExponential, Min=@n4, Max=Add(A=@n4, B=ScalarParameter() {defaultValue=0.02, parameterName=EdgeSoftness})) {materialFunction=&SmoothStep}
  n6 = TextureCoordinate()
  n7 = VectorParameter() {parameterName=Offset, defaultValue=<rgba 0 0 0 1>}
  n8 = AppendVector(A=@n7.R, B=@n7.R)
  n9 = Constant2Vector() {g=1, r=1}
  n10 = ScalarParameter() {defaultValue=1, parameterName=Scale}
  n11 = TextureSampleParameter2D(UVs=MaterialFunctionCall(UVs=Add(A=@n6, B=@n8), Texture Scale=Multiply(A=@n9, B=@n10)) {materialFunction=&ScaleUVsByCenter}.UVs) {parameterName=Texture, texture=/Game/Characters/Player/Male/Medium/Bodies/M_MED_Soccer/Textures/cRESTS/Uruguay}
  StaticSwitchParameter(True=LinearInterpolate(B=OneMinus(None=@n3.Result), Alpha=@n5.Result), False=Multiply(A=@n11.A, B=@n5.Result)) {parameterName=IsEmpty}
  StaticSwitchParameter(True=Constant() {r=1}, False=@n11.RGB) {parameterName=IsEmpty}
 ]
}
