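# M_UI_Pie is a UI material function, filed under event leaderboards, that outputs a soft-edged pie wedge mask between Start and End values.
# It shifts UVs by Offset (only R, applied to both axes), rotates them by negative Start, then converts the centered coordinates to an inverted radial angle.
# An inverted SmoothStep cuts that angle at End minus Start, with EdgeSoftness feathering only that trailing edge.
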MaterialFunction /Game/UI/Events/Leaderboards/Materials/M_UI_Pie
graph {
 properties {libraryCategoriesText [UI]}
 nodes [
  n1 = MaterialFunctionCall(Float3=FunctionInput(Preview=Constant3Vector()) {bUsePreviewValueAsDefault=true, inputName=Offset, sortPriority=4}) {materialFunction=&BreakOutFloat3Components}
  n2 = FunctionInput(Preview=Constant()) {bUsePreviewValueAsDefault=true, inputName=Start, inputType=FunctionInput_Scalar}
  n3 = MaterialFunctionCall(UVs=Add(A=TextureCoordinate(), B=AppendVector(A=@n1.R, B=@n1.R)), "Rotation Angle (0-1)"=Multiply(A=@n2) {constB=-1}) {materialFunction=&CustomRotator}
  n4 = MaterialFunctionCall(XY=Multiply(A=Subtract(A=@n3.Rotated Values) {constB=0.5}) {constB=-1}) {materialFunction=&Swizzle}
  n5 = OneMinus(None=Add(
    A=MaterialFunctionCall(Vector or UVs=@n4.YX) {materialFunction=&VectorToRadialValue}.Vector Converted to Angle
    B=Constant()
   ))
  n6 = Subtract(A=FunctionInput(Preview=Constant() {r=0.25}) {inputName=End, inputType=FunctionInput_Scalar, sortPriority=1}, B=@n2)
  n7 = Add(
   A=@n6
   B=FunctionInput(Preview=Constant() {r=0.002}) {bUsePreviewValueAsDefault=true, inputName=EdgeSoftness, inputType=FunctionInput_Scalar, sortPriority=2}
  )
  FunctionOutput(None=OneMinus(None=MaterialFunctionCall(Alpha=@n5, Min=@n6, Max=@n7) {materialFunction=&SmoothStep}.Result))
 ]
}
