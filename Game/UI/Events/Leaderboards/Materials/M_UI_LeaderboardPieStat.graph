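# Translucent UI material for leaderboard stats that draws a two-segment donut chart with parameterized percentages and colors.
# M_UI_Pie calls build consecutive angular segments, scaled by a Reveal parameter, with gaps that vanish when a percentage is zero via Ceil.
# The second segment's mask blends the two colors.
# The combined segment mask is cut into a ring by SmoothStep over a radial gradient, with adjustable ring sizes and edge softness.
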Material /Game/UI/Events/Leaderboards/Materials/M_UI_LeaderboardPieStat
graph {
 properties {
  bUsedWithStaticMesh=false
  blendMode=BLEND_Translucent
  materialDomain=MD_UI
  pixelDepthOffsetMode=PDOM_Legacy
  usedShadingModels=Default Lit
 }
 nodes [
  n1 = ScalarParameter() {defaultValue=0.75, parameterName=Percent1}
  n2 = Clamp(None=Multiply(A=@n1, B=ScalarParameter() {defaultValue=1, parameterName=Reveal, sliderMax=1}))
  n3 = ScalarParameter() {defaultValue=0.005, parameterName=PercentGap}
  n4 = Ceil(None=@n1)
  n5 = ScalarParameter() {defaultValue=0.25, parameterName=Percent2}
  n6 = Add(A=@n2, B=Multiply(A=@n5, B=ScalarParameter() {defaultValue=1, parameterName=Reveal, sliderMax=1}))
  n7 = ScalarParameter() {defaultValue=0.005, parameterName=PercentGap}
  n8 = Ceil(None=@n5)
  n9 = MaterialFunctionCall!(
   Start=Add(A=@n2, B=Multiply(A=@n3, B=@n4))
   End=Subtract(A=Clamp(None=@n6), B=Multiply(A=@n7, B=@n8))
   EdgeSoftness=ScalarParameter() {parameterName=EdgeSoftness}
   Offset
  ) {materialFunction=&/M_UI_Pie}
  LinearInterpolate(
   A=VectorParameter() {parameterName=Percent1_Color, defaultValue=<rgba 0.008 1 0.2384 1>}.RGB
   B=VectorParameter() {parameterName=Percent2_Color, defaultValue=<rgba 0.0185 1 0.9047 1>}.RGB
   Alpha=@n9.Result
  )
  n10 = MaterialFunctionCall(End=@n2, EdgeSoftness=ScalarParameter() {parameterName=EdgeSoftness}) {materialFunction=&M_UI_Pie}
  n11 = MaterialFunctionCall() {materialFunction=&RadialGradientExponential}
  n12 = OneMinus(None=ScalarParameter() {defaultValue=1, parameterName=OuterRingSize})
  n13 = Add(A=@n12, B=ScalarParameter() {parameterName=EdgeSoftness})
  n14 = Add(A=@n12, B=ScalarParameter() {defaultValue=0.2, parameterName=InnerRingSize})
  n15 = Add(A=@n14, B=ScalarParameter() {parameterName=EdgeSoftness})
  Multiply(
   A=Multiply(
    A=Max(A=@n10.Result, B=@n9.Result)
    B=MaterialFunctionCall(Alpha=@n11.RadialGradientExponential, Min=@n12, Max=@n13) {materialFunction=&SmoothStep}.Result
   )
   B=OneMinus(None=MaterialFunctionCall(Alpha=@n11.RadialGradientExponential, Min=@n14, Max=@n15) {materialFunction=&SmoothStep}.Result)
  )
 ]
}
