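# Translucent UI material for a stat icon on event leaderboards, defaulting to a broadcast eliminations texture.
# A Scale parameter feeds ScaleUVsByCenter, so the Icon texture parameter is sampled with UVs scaled around the center.
# The icon's alpha channel is inverted through OneMinus, alongside a constant white color node.
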Material /Game/UI/Events/Leaderboards/Materials/M_UI_Leaderboard_StatIUcon
graph {
 properties {
  bUsedWithStaticMesh=false
  blendMode=BLEND_Translucent
  materialDomain=MD_UI
  pixelDepthOffsetMode=PDOM_Legacy
  usedShadingModels=Default Lit
 }
 nodes [
  Constant3Vector() {constant=<rgba 1 1 1 0>}
  n1 = ScalarParameter() {defaultValue=1, parameterName=Scale}
  OneMinus(None=TextureSampleParameter2D(UVs=MaterialFunctionCall(Texture Scale=@n1) {materialFunction=&ScaleUVsByCenter}.UVs) {parameterName=Icon, texture=/Game/Athena/HUD/Spectator/Art/T-Icon-BroadcastEliminations}.A)
 ]
}
