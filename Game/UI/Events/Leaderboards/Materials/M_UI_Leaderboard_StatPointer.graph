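# M_UI_Leaderboard_StatPointer is a translucent UI material that builds a stat pointer shape for event leaderboards from a dot and a line.
# A radial gradient on scaled UVs, with parameterized center and radius, passes through SmoothStep to form the circle.
# Two SmoothSteps on the V coordinate are subtracted to cut a soft horizontal band around the middle, sized by LineWidth.
# Max merges both procedural masks without textures, with EdgeSoftness controlling their edges.
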Material /Game/UI/Events/Leaderboards/Materials/M_UI_Leaderboard_StatPointer
graph {
 properties {
  bUsedWithStaticMesh=false
  blendMode=BLEND_Translucent
  materialDomain=MD_UI
  pixelDepthOffsetMode=PDOM_Legacy
  usedShadingModels=Default Lit
 }
 nodes [
  Constant() {r=1}
  n1 = VectorParameter() {parameterName=CircleSize, defaultValue=<rgba 1 1 0 1>}
  n2 = Multiply(A=TextureCoordinate(), B=AppendVector(A=@n1.R, B=@n1.R))
  n3 = VectorParameter() {parameterName=CenterPosition, defaultValue=<rgba 0.5 0.5 0 1>}
  n4 = AppendVector(A=@n3.R, B=@n3.R)
  n5 = ScalarParameter() {defaultValue=0.6476, parameterName=Radius}
  n6 = Constant()
  n7 = ScalarParameter() {defaultValue=0.01, parameterName=EdgeSoftness}
  n8 = ComponentMask(None=TextureCoordinate()) {g=true}
  n9 = Subtract(A=Constant() {r=0.5}, B=ScalarParameter() {defaultValue=0.1, parameterName=LineWidth})
  n10 = Add(A=@n9, B=ScalarParameter() {defaultValue=0.01, parameterName=EdgeSoftness})
  n11 = Add(A=Constant() {r=0.5}, B=ScalarParameter() {defaultValue=0.1, parameterName=LineWidth})
  n12 = Add(A=@n11, B=ScalarParameter() {defaultValue=0.01, parameterName=EdgeSoftness})
  Max(
   A=MaterialFunctionCall(
    Alpha=MaterialFunctionCall(UVs=@n2, CenterPosition=@n4, Radius=@n5) {materialFunction=&RadialGradientExponential}.RadialGradientExponential
    Min=@n6
    Max=Add(A=@n6, B=@n7)
   ) {materialFunction=&SmoothStep}.Result
   B=Subtract(
    A=MaterialFunctionCall(Alpha=@n8, Min=@n9, Max=@n10) {materialFunction=&SmoothStep}.Result
    B=MaterialFunctionCall(Alpha=@n8, Min=@n11, Max=@n12) {materialFunction=&SmoothStep}.Result
   )
  )
 ]
}
