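# Translucent UI material for the tail of a leaderboard stat pointer, drawing a soft-edged straight band at an adjustable angle.
# It rotates the UVs around the middle of the left edge by the Angle parameter and takes the vertical coordinate.
# Two SmoothSteps offset either side of center by LineWidth are subtracted to form the line mask procedurally, without textures, with EdgeSoftness setting the falloff.
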Material /Game/UI/Events/Leaderboards/Materials/M_UI_Leaderboard_StatPointer_Tail
graph {
 properties {
  bUsedWithStaticMesh=false
  blendMode=BLEND_Translucent
  materialDomain=MD_UI
  pixelDepthOffsetMode=PDOM_Legacy
  usedShadingModels=Default Lit
 }
 nodes [
  Constant() {r=1}
  n1 = ComponentMask(None=MaterialFunctionCall(UVs=TextureCoordinate(), Rotation Center=Constant2Vector() {g=0.5}, "Rotation Angle (0-1)"=ScalarParameter() {parameterName=Angle}) {materialFunction=&CustomRotator}.Rotated Values) {g=true}
  n2 = Subtract(A=Constant() {r=0.5}, B=ScalarParameter() {defaultValue=0.1, parameterName=LineWidth})
  n3 = ScalarParameter() {defaultValue=0.01, parameterName=EdgeSoftness}
  n4 = Add(A=Constant() {r=0.5}, B=ScalarParameter() {defaultValue=0.1, parameterName=LineWidth})
  n5 = ScalarParameter() {defaultValue=0.01, parameterName=EdgeSoftness}
  Subtract(
   A=MaterialFunctionCall(Alpha=@n1, Min=@n2, Max=Add(A=@n2, B=@n3)) {materialFunction=&SmoothStep}.Result
   B=MaterialFunctionCall(Alpha=@n1, Min=@n4, Max=Add(A=@n4, B=@n5)) {materialFunction=&SmoothStep}.Result
  )
 ]
}
