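# Additive UI material for the event leaderboard live icon, drawing animated broadcast-style ping arcs.
# Two MF_UI_Ring pulses share a time loop with different easing exponents, expanding as they thin, soften and fade, then combine by maximum.
# An angle gradient, from a deprecated function, cuts them into left and right sweeps, and an inverted radial gradient masks the center.
# A Color parameter is also defined.
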Material /Game/UI/Events/Leaderboards/Materials/M_UI_LiveIcon
graph {
 properties {bUsedWithStaticMesh=false, blendMode=BLEND_Additive, materialDomain=MD_UI, pixelDepthOffsetMode=PDOM_Legacy, usedShadingModels=Default Lit}
 nodes [
  n1 = ScalarParameter() {defaultValue=1, group=Ping, parameterName=PingSpeed}
  n2 = Fmod(A=Multiply(A=Time() {period=1}, B=@n1) {constB=0.5}, B=Constant() {r=1})
  Add(A=Multiply(A=@n2) {constB=0.5}) {constB=0.5}
  n3 = MaterialFunctionCall() {materialFunction=&RadialGradientExponential}
  n4 = ScalarParameter() {defaultValue=0.8, parameterName=InnerMask}
  n5 = Add(A=@n4, B=ScalarParameter() {defaultValue=0.002, parameterName=EdgeSoftness})
  n6 = ScalarParameter() {defaultValue=0.1, group=Ping, parameterName=PingRingMaxSize}
  n7 = Power(Base=@n2) {constExponent=0.4}
  n8 = LinearInterpolate(A=@n6, Alpha=@n7) {constA=0.1, constB=0.01}
  n9 = LinearInterpolate(Alpha=@n7) {constA=0.002, constB=0.05}
  n10 = ScalarParameter() {defaultValue=0.5, group=Ping, parameterName=PingOuterRadius}
  n11 = LinearInterpolate(B=@n10, Alpha=@n7) {constB=0.5}
  n12 = Clamp(None=Multiply(
    A=MaterialFunctionCall!(RingSize=@n8, Edge, EdgeSoftness=@n9, OuterRadius=@n11) {materialFunction=/Game/UI/InGame/HUD/Textures/MF_UI_Ring}.Result
    B=LinearInterpolate(Alpha=@n7) {constA=2, constB=0}
   ))
  n13 = Time() {period=1}
  n14 = Power(Base=Fmod(A=Multiply(A=@n13, B=@n1) {constB=0.5}, B=Constant() {r=1})) {constExponent=1.1}
  n15 = LinearInterpolate(A=@n6, Alpha=@n14) {constA=0.1, constB=0.01}
  n16 = LinearInterpolate(Alpha=@n14) {constA=0.002, constB=0.05}
  n17 = LinearInterpolate(B=@n10, Alpha=@n14) {constB=0.5}
  n18 = Clamp(None=Multiply(
    A=MaterialFunctionCall(RingSize=@n15, EdgeSoftness=@n16, OuterRadius=@n17) {materialFunction=&MF_UI_Ring}.Result
    B=LinearInterpolate(Alpha=@n14) {constA=2, constB=0}
   ))
  n19 = MaterialFunctionCall!(UVs, Rotation) {materialFunction=/Game/UI/Foundation/Materials/MF_UI_AngleGradient_DEPRECATED}
  n20 = ScalarParameter() {defaultValue=0.8, parameterName=RightAngle}
  n21 = ScalarParameter() {parameterName=EdgeSoftness}
  n22 = Add(A=@n20, B=ScalarParameter() {defaultValue=0.2, parameterName=Sweep})
  n23 = ScalarParameter() {defaultValue=0.002, parameterName=EdgeSoftness}
  n24 = Subtract(
   A=MaterialFunctionCall(Alpha=@n19.Result, Min=@n20, Max=Add(A=@n21, B=@n20)) {materialFunction=&SmoothStep}.Result
   B=MaterialFunctionCall(Alpha=@n19.Result, Min=@n22, Max=Add(A=@n22, B=@n23)) {materialFunction=&SmoothStep}.Result
  )
  n25 = ScalarParameter() {defaultValue=0.2, parameterName=LeftAngle}
  n26 = ScalarParameter() {defaultValue=0.002, parameterName=EdgeSoftness}
  n27 = Add(A=@n25, B=ScalarParameter() {defaultValue=0.2, parameterName=Sweep})
  n28 = ScalarParameter() {defaultValue=0.002, parameterName=EdgeSoftness}
  n29 = Subtract(
   A=MaterialFunctionCall(Alpha=@n19.Result, Min=@n25, Max=Add(A=@n26, B=@n25)) {materialFunction=&SmoothStep}.Result
   B=MaterialFunctionCall(Alpha=@n19.Result, Min=@n27, Max=Add(A=@n27, B=@n28)) {materialFunction=&SmoothStep}.Result
  )
  Multiply(
   A=OneMinus(None=MaterialFunctionCall(Alpha=@n3.RadialGradientExponential, Min=@n4, Max=@n5) {materialFunction=&SmoothStep}.Result)
   B=Multiply(A=Max(A=@n12, B=@n18), B=Max(A=@n24, B=@n29))
  )
  VectorParameter() {parameterName=Color, defaultValue=<rgba 1 1 1 1>}
 ]
}
