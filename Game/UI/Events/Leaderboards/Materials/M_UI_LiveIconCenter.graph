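# M_UI_LiveIconCenter is a translucent UI material for the leaderboard events interface, named as the center of a live icon.
# A radial exponential gradient feeds a SmoothStep that runs from Edge to Edge plus EdgeSoftness, forming a procedural soft-edged circular mask without textures.
# A Color vector parameter, black by default, sits alongside the mask.
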Material /Game/UI/Events/Leaderboards/Materials/M_UI_LiveIconCenter
graph {
 properties {
  bUsedWithStaticMesh=false
  blendMode=BLEND_Translucent
  materialDomain=MD_UI
  pixelDepthOffsetMode=PDOM_Legacy
  usedShadingModels=Default Lit
 }
 nodes [
  n1 = ScalarParameter() {defaultValue=0.2, parameterName=Edge}
  MaterialFunctionCall(
   Alpha=MaterialFunctionCall() {materialFunction=&RadialGradientExponential}.RadialGradientExponential
   Min=@n1
   Max=Add(A=@n1, B=ScalarParameter() {defaultValue=0.05, parameterName=EdgeSoftness})
  ) {materialFunction=&SmoothStep}
  VectorParameter() {parameterName=Color, defaultValue=<rgba 0 0 0 1>}
 ]
}
