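# TeamEntry is a UserWidget for a team entry in the event leaderboards UI, showing text in a single styled label.
# A SizeBox root holds a CommonTextBlock, Text_Team, set in Burbank with a scroll style and "Username" as placeholder text.
# Its SetText function takes a Usernames text input and writes it to that label; the event graph is empty.
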WidgetBlueprint /Game/UI/Events/Leaderboards/TeamEntry
graphs [
 EventGraph {links [], nodes []}
 SetText {nodes [
   n1 = SetText!() -> (then, Usernames:Text) {is=title K2Node_FunctionEntry}
   > Set Text(Text=@n1, self=Get Text_Team() -> (Text_Team:Common Text Object Reference))
  ]}
]
variables []
widgets {
 info {inheritedWidgetCount=0, parentClass=&UserWidget, rootWidgetClass=&SizeBox, widgetCount=2}
 widgets [
  SizeBox_0 {widgetClassPath=&SizeBox, widget=SizeBox_0}
  Text_Team {
   widgetClassPath=&CommonTextBlock
   bIsVariable=true
   parent=SizeBox_0
   widget=Text_Team
   properties {
    scrollStyle=&TextScrollStyle-Base_C
    text=Username
    font {
     bForceMonospaced=false
     bMaterialIsStencil=false
     fontMaterial=null
     fontObject=&Burbank
     letterSpacing=0
     monospacedWidth=1
     size=24
     skewAmount=0
     typefaceFontName=Burbank-PlayerPower
     outlineSettings {
      bApplyOutlineToDropShadows=false
      bMiteredCorners=false
      bSeparateFillAlpha=false
      outlineMaterial=null
      outlineSize=0
      outlineColor=<rgba 0 0 0 1>
     }
    }
    shadowOffset=<0 0>
   }
   slot {class=SizeBoxSlot}
  }
 ]
}
