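# Event leaderboard row widget showing rank, team names, score, flag, qualification icon and live-session indicator.
# SetEntryDetails fills it from the list item's entry data, enlarging first place and setting zebra and match-point state for EventStyleEntry.
# EventColorize, uncalled here, pulls tournament colors from the parent FortShowdownDetailView; selection plays a highlight animation forward or reverse.
# A random-flag chain using a TEMP identity table is left unconnected.
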WidgetBlueprint /Game/UI/Events/Leaderboards/EventLeaderboardEntry
animations [Selected {
  binding_count=2
  length_seconds=0.3292
  object=&self:Selected
  sequence {bindings [
    CommonBorder_SelectedTint {
     id=#1
     possessed_class=CommonBorder
     tracks [{
       type=MovieSceneColorTrack
       display_name=Content Color and Opacity
       sections [{
         type=MovieSceneColorSection
         end_frame=6
         start_frame=0
         channels [
          Color.R {type=MovieSceneScriptingFloatChannel, keys [[0,1],[6.5843,0]]}
          Color.G {type=MovieSceneScriptingFloatChannel, keys [[0,1],[6.5843,0]]}
          Color.B {type=MovieSceneScriptingFloatChannel, keys [[0,1],[6.5843,0]]}
          Color.A {type=MovieSceneScriptingFloatChannel, const [1,2]}
         ]
        }]
      }]
    }
    Image_SelectedBG {
     id=#2
     possessed_class=Image
     tracks [{
       type=MovieSceneFloatTrack
       display_name=Render Opacity
       sections [{type=MovieSceneFloatSection, end_frame=6, start_frame=0, channels [None {type=MovieSceneScriptingFloatChannel, keys [[0,0],[6.5843,1]]}]}]
      }]
    }
   ]}
 }]
graphs [
 EventGraph {nodes [
   EventStyleEntry() {is=title K2Node_CustomEvent}
   n1 = AND Boolean(A=Get isZebra() -> (isZebra:bool), B=NOT Boolean(A=Get isPlayer() -> (isPlayer:bool)))
   n2 = Make Linear Color!(R:float=0.01, G:float=0, B:float=0.15, A:float=0.2) -> (LinearColor:Linear Color Structure) {is=title K2Node_MakeStruct}
   > SetBrushColor(
    self=Get CommonBorder_Zebra() -> (CommonBorder_Zebra:Common Border Object Reference)
    InBrushColor=Select(
     Option 0:Linear Color Structure=Select(Option 0=(R=0,G=0,B=0,A=0), Option 1=@n2, Index=@n1)
     Option 1:Linear Color Structure=(R=0.8021,G=0.5848,B=0,A=1)
     Index=Get isMatchPoint() -> (isMatchPoint:bool)
    ) -> (ReturnValue:Linear Color Structure)
   )
   EventColorize() {is=title K2Node_CustomEvent}
   > n3 = FindParentWidgetOfType!(execute, StartingWidget:Widget Object Reference=Self-Reference(), Type:Widget Class Reference=/Script/FortniteUI.FortShowdownDetailView) -> (then, ReturnValue:Fort Showdown Detail View Object Reference)
   > n4 = Cast To FortShowdownDetailView!(execute, Object=@n3) -> (then, CastFailed, AsFort Showdown Detail View:Fort Showdown Detail View Object Reference) {is=title K2Node_DynamicCast}
   n5 = Break Fort Tournament Display Info!(FortTournamentDisplayInfo:"Fort Tournament Display Info Structure (by ref)"=Get TournamentDisplayInfo!(self:Fort Showdown Detail View Object Reference=@n4) -> (TournamentDisplayInfo:Fort Tournament Display Info Structure)) -> (
    TitleLine1:Text
    TitleLine2:Text
    ScheduleInfo:Text
    FlavorDescription:Text
    DetailsDescription:Text
    ShortFormatTitle:Text
    LongFormatTitle:Text
    BackgroundTitle:Text
    PinScoreRequirement:int
    PinEarnedText:Text
    BaseColor:Linear Color Structure
    PrimaryColor:Linear Color Structure
    SecondaryColor:Linear Color Structure
    HighlightColor:Linear Color Structure
    TitleColor:Linear Color Structure
    ShadowColor:Linear Color Structure
    BackgroundLeftColor:Linear Color Structure
    BackgroundRightColor:Linear Color Structure
    BackgroundTextColor:Linear Color Structure
    PosterFadeColor:Linear Color Structure
    AlertText:Text
    SeriesPointLeaderboardName:Text
    AlertType:EFortTournamentAlertType Enum
    RoundDisplayNames:Array of Texts
   ) {is=title K2Node_BreakStruct}
   > SetColorAndOpacity(self=Get Text_Rank() -> (Text_Rank:Common Text Object Reference), InColorAndOpacity=Make Slate Color(SpecifiedColor=@n5.SecondaryColor))
   n6 = BreakColor(InColor=@n5.PrimaryColor)
   > SetBrushColor(
    self=Get CommonBorder_Zebra()
    InBrushColor=Select(
     Option 0=(R=0,G=0,B=0,A=0)
     Option 1=MakeColor!(R:float=@n6, G:float=@n6.G, B:float=@n6.B, A:float=0.06) -> (ReturnValue:Linear Color Structure)
     Index=Get isZebra()
    )
   )
   > SetBrushColor(
    self=Get CommonBorder_IsPlayer() -> (CommonBorder_IsPlayer:Common Border Object Reference)
    InBrushColor=Select(Option 0=(R=0,G=0,B=0,A=0), Option 1=@n5.HighlightColor, Index=Get isPlayer())
   )
   Event On Selected()
   n7 = Get Selected() -> (Selected:Widget Animation Object Reference)
   > PlayAnimationForward(InAnimation=@n7, PlaybackSpeed=4)
   Event On Deselected()
   > PlayAnimationReverse(InAnimation=@n7, PlaybackSpeed=3)
   n8 = Event InitializeTeamTextEntry!() -> (OutputDelegate, then, TeamEntryWidget:User Widget Object Reference, Usernames:"Text (by ref)") {is=title K2Node_Event}
   > n9 = Cast To TeamEntry!(execute, Object=@n8.TeamEntryWidget) -> (then, CastFailed, AsTeam Entry:Team Entry Object Reference) {is=title K2Node_DynamicCast}
   > SetText!(execute, self:Team Entry Object Reference=@n9, Usernames:Text=@n8.Usernames) -> (then)
   n10 = Event OnListItemObjectSet()
   > n11 = Cast To FortEventLeaderboardEntryData!(execute, Object=@n10.ListItemObject) -> (then, CastFailed, AsFort Event Leaderboard Entry Data:Fort Event Leaderboard Entry Data Object Reference) {is=title K2Node_DynamicCast}
   > SetEntryDetails!(execute, self, EntryData:Fort Event Leaderboard Entry Data Object Reference=@n11) -> (then)
   Event PreConstruct()
   > EventStyleEntry() {is=title K2Node_CallFunction}
   n12 = GetParent!(self:Widget Object Reference) -> (ReturnValue:Panel Widget Object Reference)
   n13 = Self-Reference()
   "Equal (Integer)"(
    A="% (Integer)"(A=GetChildIndex!(self:Panel Widget Object Reference=@n12, Content:Widget Object Reference=@n13) -> (ReturnValue:int=0), B=2) -> (ReturnValue=0)
    B=0
   ) {is=title K2Node_CallFunction}
  ]}
 SetEntryDetails {nodes [
   n1 = Get TEMPDatatableIdentities() -> (TEMPDatatableIdentities:Data Table Soft Object Reference)
   n2 = GetDataTableRowNames(Table=Resolve Soft Reference!(Input:Data Table Soft Object Reference=@n1) -> (Output:Data Table Object Reference))
   n3 = int - int(A=Length(TargetArray:Array of Names=@n2), B=1) {is=title K2Node_CallFunction}
   > n4 = Get Data Table Row!(
    execute:Exec
    DataTable:Data Table Object Reference=Resolve Soft Reference(Input=@n1)
    RowName:Name="Get (a copy)"(Array:Array of Names=@n2, Dimension 1=RandomIntegerInRange!(Min:int=0, Max:int=@n3) -> (ReturnValue:int=0)) -> (Output:Name)
   ) -> (then:Exec, RowNotFound:Exec, ReturnValue:Fort Competitive Identity Info Structure)
   > SetPlayerFlag!(
    execute
    self:Fort Event Leaderboard Entry Object Reference
    PlayerFlag:Texture 2D Soft Object Reference=Break Fort Competitive Identity Info!(FortCompetitiveIdentityInfo:"Fort Competitive Identity Info Structure (by ref)"=@n4) -> (RegionId:String, DisplayName:Text, FlagAsset:Texture 2D Soft Object Reference) {is=title K2Node_BreakStruct}.FlagAsset
   ) -> (then)
   n5 = SetEntryDetails!() -> (then, EntryData:Fort Event Leaderboard Entry Data Object Reference) {is=title K2Node_FunctionEntry}
   > n6 = Sequence!(execute) -> (then_0, then_1, then_2:Exec, then_3:Exec, then_4:Exec, then_5:Exec)
   n7 = GetRank!(self:Fort Event Leaderboard Entry Data Object Reference=@n5) -> (ReturnValue:int=0)
   > "SetText (Text)"(self=Get Text_Rank(), InText=Format Text!(Format="LOCTEXT(\"##{rank}\")", rank:Text="To Text (Integer)"(Value=@n7)) -> (Result))
   n8 = GetPageIndex!(self:Fort Event Leaderboard Entry Data Object Reference=@n5) -> (ReturnValue:int=0)
   > Set isZebra(isZebra:bool="Equal (Integer)"(A="% (Integer)"(A=@n8, B=2) -> (ReturnValue=0), B=0)) -> (Output_Get:bool)
   > Set isMatchPoint(isMatchPoint:bool=GetIsMatchPointWinner!(self:Fort Event Leaderboard Entry Data Object Reference=@n5) -> (ReturnValue:bool)) -> (Output_Get:bool)
   n9 = "Equal (Integer)"(A=@n7, B=1)
   > SetActiveWidgetIndex(
    self=Get CommonWidgetSwitcher_Rank() -> (CommonWidgetSwitcher_Rank:Common Widget Switcher Legacy Object Reference)
    Index=Select(Option 0:int=0, Option 1:int=1, Index=@n9) -> (ReturnValue:int)
   )
   > SetPadding!(
    execute
    self:Border Object Reference=Get CommonBorder_IsPlayer()
    InPadding:Margin Structure=Make Margin(Left=0, Top=Select(Option 0:double=0, Option 1:double=20, Index=@n9) -> (ReturnValue:double), Right=0, Bottom=0)
   ) -> (then)
   > "SetText (Text)"(
    self=Get Text_Score() -> (Text_Score:Common Text Object Reference)
    InText="To Text (Integer)"(Value=GetScore!(self:Fort Event Leaderboard Entry Data Object Reference=@n5) -> (ReturnValue:int=0))
   )
   DisplayTeamUsernames!(
    execute=@n6.then_1
    self:Fort Event Leaderboard Entry Object Reference
    LeaderboardEntryData:Fort Event Leaderboard Entry Data Object Reference=@n5
   ) -> (then)
   n10 = GetLiveSessionId!(self:Fort Event Leaderboard Entry Data Object Reference=@n5) -> (ReturnValue:String)
   n11 = Branch(execute=@n6.then_3, Condition=TextIsEmpty(InText="To Text (String)"(InString=@n10)))
   n12 = Get Overlay_OnlineStatus() -> (Overlay_OnlineStatus:Overlay Object Reference)
   > SetVisibility(self=@n12, InVisibility=Collapsed)
   SetVisibility(execute=@n11.else, self=@n12, InVisibility=SelfHitTestInvisible)
   EventStyleEntry(execute=@n6.then_2)
   SetVisibility(
    execute=@n6.then_4
    self=Get Overlay_QualifiedContent() -> (Overlay_QualifiedContent:Overlay Object Reference)
    InVisibility=Select(
     Option 0:ESlateVisibility Enum=Collapsed
     Option 1:ESlateVisibility Enum=SelfHitTestInvisible
     Index=DoesLeaderboardHaveQualifiedIcons!(self:Fort Event Leaderboard Entry Data Object Reference=@n5) -> (ReturnValue:bool)
    ) -> (ReturnValue:ESlateVisibility Enum)
   )
   > SetVisibility(
    self=Get Image_Qualified() -> (Image_Qualified:Image Object Reference)
    InVisibility=Select(
     Option 0=Hidden
     Option 1=SelfHitTestInvisible
     Index=ShouldShowQualifiedIcon!(self:Fort Event Leaderboard Entry Data Object Reference=@n5) -> (ReturnValue:bool)
    )
   )
  ]}
]
variables [
 isMatchPoint {type=boolean}
 isPlayer {type=boolean}
 isZebra {type=boolean, default=true}
 TEMPDatatableIdentities {type=DataTable, default=/Game/UI/Competitive/CompetitiveIdentity}
]
widgets {
 info {
  inheritedWidgetCount=3
  parentClass=/Script/FortniteUI.FortEventLeaderboardEntry
  rootWidgetClass=&CommonBorder
  widgetCount=31
 }
 widgets [
  CommonBorder_IsPlayer {
   widgetClassPath=&CommonBorder
   bIsVariable=true
   widget=CommonBorder_IsPlayer
   properties {
    background {
     bIsDynamicallyLoaded=false
     drawAs=Image
     imageType=NoImage
     mirroring=NoMirror
     resourceName=None
     resourceObject=null
     tiling=NoTile
     imageSize=<32 32>
     margin=<ltrb 0 0 0 0>
     outlineSettings {
      bUseBrushTransparency=false
      roundingType=HalfHeightRadius
      width=0
      color {colorUseRule=UseColor_Specified, specifiedColor=<rgba 0 0 0 0>}
      cornerRadii=<0 0 0 0>
     }
     tintColor {colorUseRule=UseColor_Specified, specifiedColor=<rgba 1 1 1 1>}
     uVRegion {bIsValid=false, max=<0 0>, min=<0 0>}
    }
    brushColor=<rgba 0.0009 0.1812 0.8796 0>
    padding=<ltrb 0 0 0 0>
   }
  }
  CommonBorder_Zebra {
   widgetClassPath=&CommonBorder
   bIsVariable=true
   parent=CommonBorder_IsPlayer
   widget=CommonBorder_Zebra
   properties {
    background {
     bIsDynamicallyLoaded=false
     drawAs=Image
     imageType=NoImage
     mirroring=NoMirror
     resourceName=None
     resourceObject=null
     tiling=NoTile
     imageSize=<32 32>
     margin=<ltrb 0 0 0 0>
     outlineSettings {
      bUseBrushTransparency=false
      roundingType=HalfHeightRadius
      width=0
      color {colorUseRule=UseColor_Specified, specifiedColor=<rgba 0 0 0 0>}
      cornerRadii=<0 0 0 0>
     }
     tintColor {colorUseRule=UseColor_Specified, specifiedColor=<rgba 1 1 1 1>}
     uVRegion {bIsValid=false, max=<0 0>, min=<0 0>}
    }
    brushColor=<rgba 0.0097 0 0.9047 0.0588>
    padding=<ltrb 0 0 0 0>
   }
   slot {class=BorderSlot, properties {padding=<ltrb 0 0 0 0>}}
  }
  HorizontalBox_0 {
   widgetClassPath=&HorizontalBox
   parent=CommonBorder_Zebra
   widget=HorizontalBox_0
   slot {class=BorderSlot, properties {padding=<ltrb 0 0 0 0>}}
  }
  SizeBox_1 {
   widgetClassPath=&SizeBox
   parent=HorizontalBox_0
   widget=SizeBox_1
   properties {bOverride_WidthOverride=true, widthOverride=68.3351}
   slot {class=HorizontalBoxSlot, properties {verticalAlignment=VAlign_Center, padding=<ltrb 5 0 0 0>}}
  }
  CommonWidgetSwitcher_Rank {
   widgetClassPath=&CommonWidgetSwitcherLegacy
   bIsVariable=true
   parent=SizeBox_1
   widget=CommonWidgetSwitcher_Rank
   properties {transitionDuration=0}
   slot {class=SizeBoxSlot}
  }
  Text_Rank {
   widgetClassPath=&CommonTextBlock
   bIsVariable=true
   parent=CommonWidgetSwitcher_Rank
   widget=Text_Rank
   properties {
    text="##2"
    colorAndOpacity {colorUseRule=UseColor_Specified, specifiedColor=<rgba 1 1 1 0.6>}
    font {
     bForceMonospaced=false
     bMaterialIsStencil=false
     fontMaterial=null
     fontObject=&Burbank
     letterSpacing=0
     monospacedWidth=1
     size=24
     skewAmount=0
     typefaceFontName=Burbank-PlayerPower
     outlineSettings {
      bApplyOutlineToDropShadows=false
      bMiteredCorners=false
      bSeparateFillAlpha=false
      outlineMaterial=null
      outlineSize=0
      outlineColor=<rgba 0 0 0 1>
     }
    }
    shadowOffset=<0 0>
   }
   slot {class=WidgetSwitcherSlot, properties {horizontalAlignment=HAlign_Right}}
  }
  Text_Rank_First {
   widgetClassPath=&CommonTextBlock
   bIsVariable=true
   parent=CommonWidgetSwitcher_Rank
   widget=Text_Rank_First
   properties {
    text="##1"
    font {
     bForceMonospaced=false
     bMaterialIsStencil=false
     fontMaterial=null
     fontObject=&Burbank
     letterSpacing=0
     monospacedWidth=1
     size=42
     skewAmount=0
     typefaceFontName=BurbankRegular-Black
     outlineSettings {
      bApplyOutlineToDropShadows=false
      bMiteredCorners=false
      bSeparateFillAlpha=false
      outlineMaterial=null
      outlineSize=0
      outlineColor=<rgba 0 0 0 1>
     }
    }
    shadowOffset=<0 0>
   }
   slot {class=WidgetSwitcherSlot, properties {horizontalAlignment=HAlign_Right}}
  }
  CommonBorder_0 {
   widgetClassPath=&CommonBorder
   parent=CommonWidgetSwitcher_Rank
   widget=CommonBorder_0
   properties {verticalAlignment=VAlign_Center, padding=<ltrb 0 -30 0 -30>}
   slot {class=WidgetSwitcherSlot}
  }
  SizeBox_0 {
   widgetClassPath=&SizeBox
   parent=CommonBorder_0
   widget=SizeBox_0
   properties {bOverride_HeightOverride=true, bOverride_WidthOverride=true, heightOverride=112.7939}
   slot {class=BorderSlot, properties {verticalAlignment=VAlign_Center, padding=<ltrb 0 -30 0 -30>}}
  }
  ScaleBox_0 {
   widgetClassPath=&ScaleBox
   parent=SizeBox_0
   widget=ScaleBox_0
   properties {stretchDirection=DownOnly, userSpecifiedScale=0.4524}
   slot {class=SizeBoxSlot, properties {horizontalAlignment=HAlign_Right}}
  }
  Overlay_5 {
   widgetClassPath=&Overlay
   parent=ScaleBox_0
   widget=Overlay_5
   properties {renderTransform {angle=-12, scale=<1 1>, shear=<-10 0>, translation=<0 0>}}
   slot {class=ScaleBoxSlot, properties {horizontalAlignment=HAlign_Right}}
  }
  Border_2 {
   widgetClassPath=&Border
   parent=Overlay_5
   widget=Border_2
   properties {
    background {
     bIsDynamicallyLoaded=false
     drawAs=NoDrawType
     imageType=NoImage
     mirroring=NoMirror
     resourceName=None
     resourceObject=null
     tiling=NoTile
     imageSize=<32 32>
     margin=<ltrb 0 0 0 0>
     outlineSettings {
      bUseBrushTransparency=false
      roundingType=HalfHeightRadius
      width=0
      color {colorUseRule=UseColor_Specified, specifiedColor=<rgba 0 0 0 0>}
      cornerRadii=<0 0 0 0>
     }
     tintColor {colorUseRule=UseColor_Specified, specifiedColor=<rgba 1 1 1 1>}
     uVRegion {bIsValid=false, max=<0 0>, min=<0 0>}
    }
    contentColorAndOpacity=<rgba 0.0123 0.0144 0.0395 1>
    padding=<ltrb 0 0 0 0>
   }
   slot {class=OverlaySlot, properties {horizontalAlignment=HAlign_Right}}
  }
  RichTextBlock_PlacementShadow {
   widgetClassPath=/Script/UMG.RichTextBlock
   parent=Border_2
   widget=RichTextBlock_PlacementShadow
   properties {text="<PlacementNumberSymbol>##</><PlacementValue>1</>", textStyleSet=/Game/Athena/HUD/EndGame/VictoryRichTextStyles}
   slot {class=BorderSlot, properties {padding=<ltrb 0 0 0 0>}}
  }
  RichTextBlock_Placement {
   widgetClassPath=&RichTextBlock
   parent=Overlay_5
   widget=RichTextBlock_Placement
   properties {
    justification=Right
    text="<PlacementNumberSymbol>##</><PlacementValue>1</>"
    textStyleSet=&VictoryRichTextStyles
    renderTransform {angle=0, scale=<1 1>, shear=<0 0>, translation=<-8 -9>}
   }
   slot {class=OverlaySlot, properties {horizontalAlignment=HAlign_Right}}
  }
  Overlay_QualifiedContent {
   widgetClassPath=&Overlay
   bIsVariable=true
   parent=HorizontalBox_0
   widget=Overlay_QualifiedContent
   slot {class=HorizontalBoxSlot, properties {horizontalAlignment=HAlign_Center, verticalAlignment=VAlign_Center, padding=<ltrb 10 0 0 0>}}
  }
  Image_Qualified {
   widgetClassPath=&Image
   bIsVariable=true
   parent=Overlay_QualifiedContent
   widget=Image_Qualified
   properties {brush {
     bIsDynamicallyLoaded=false
     drawAs=Image
     imageType=NoImage
     mirroring=NoMirror
     resourceName=None
     resourceObject=/Game/UI/Competitive/Icons/T_UI_TournamentQualifiedIcon
     tiling=NoTile
     imageSize=<64 64>
     margin=<ltrb 0 0 0 0>
     outlineSettings {
      bUseBrushTransparency=false
      roundingType=HalfHeightRadius
      width=0
      color {colorUseRule=UseColor_Specified, specifiedColor=<rgba 0 0 0 0>}
      cornerRadii=<0 0 0 0>
     }
     tintColor {colorUseRule=UseColor_Specified, specifiedColor=<rgba 1 1 1 1>}
     uVRegion {bIsValid=false, max=<0 0>, min=<0 0>}
    }}
   slot {class=OverlaySlot, properties {horizontalAlignment=HAlign_Center, verticalAlignment=VAlign_Center}}
  }
  Overlay_0 {
   widgetClassPath=&Overlay
   parent=HorizontalBox_0
   widget=Overlay_0
   slot {class=HorizontalBoxSlot, properties {padding=<ltrb 25 0 16 0>, size {sizeRule=Fill, value=1}}}
  }
  Image_SelectedBG {
   widgetClassPath=&Image
   bIsVariable=true
   parent=Overlay_0
   widget=Image_SelectedBG
   properties {
    renderOpacity=0
    brush {
     bIsDynamicallyLoaded=false
     drawAs=Box
     imageType=NoImage
     mirroring=NoMirror
     resourceName=None
     resourceObject=null
     tiling=NoTile
     imageSize=<256 64>
     margin=<ltrb 0.4 0.2 0.4 0.2>
     outlineSettings {
      bUseBrushTransparency=false
      roundingType=HalfHeightRadius
      width=0
      color {colorUseRule=UseColor_Specified, specifiedColor=<rgba 0 0 0 0>}
      cornerRadii=<0 0 0 0>
     }
     tintColor {colorUseRule=UseColor_Specified, specifiedColor=<rgba 1 1 1 1>}
     uVRegion {bIsValid=false, max=<0 0>, min=<0 0>}
    }
   }
   slot {class=OverlaySlot, properties {horizontalAlignment=HAlign_Fill, verticalAlignment=VAlign_Fill, padding=<ltrb -10 0 -15 0>}}
  }
  HorizontalBox_2 {
   widgetClassPath=&HorizontalBox
   parent=Overlay_0
   widget=HorizontalBox_2
   slot {class=OverlaySlot, properties {horizontalAlignment=HAlign_Fill, verticalAlignment=VAlign_Center, padding=<ltrb 10 0 0 0>}}
  }
  CommonBorder_SelectedTint {
   widgetClassPath=&CommonBorder
   parent=HorizontalBox_2
   widget=CommonBorder_SelectedTint
   properties {padding=<ltrb 0 0 0 0>}
   slot {class=HorizontalBoxSlot, properties {size {sizeRule=Fill, value=1}}}
  }
  HorizontalBox_1 {
   widgetClassPath=&HorizontalBox
   parent=CommonBorder_SelectedTint
   widget=HorizontalBox_1
   slot {class=BorderSlot, properties {padding=<ltrb 0 0 0 0>}}
  }
  SizeBox_3 {
   widgetClassPath=&SizeBox
   parent=HorizontalBox_1
   widget=SizeBox_3
   properties {bOverride_WidthOverride=true}
   slot {class=HorizontalBoxSlot, properties {size {sizeRule=Fill, value=1}}}
  }
  DynamicEntryBox_TeamEntries {
   widgetClassPath=&DynamicEntryBox
   bInherited=true
   bIsVariable=true
   parent=SizeBox_3
   widget=DynamicEntryBox_TeamEntries
   properties {entryBoxType=Vertical, entryWidgetClass=&/TeamEntry.TeamEntry_C, numDesignerPreviewEntries=1}
   slot {class=SizeBoxSlot}
  }
  Overlay_OnlineStatus {
   widgetClassPath=&Overlay
   bIsVariable=true
   parent=HorizontalBox_1
   widget=Overlay_OnlineStatus
   properties {visibility=Collapsed}
   slot {class=HorizontalBoxSlot, properties {horizontalAlignment=HAlign_Center, verticalAlignment=VAlign_Center}}
  }
  Image_ping {
   widgetClassPath=&Image
   parent=Overlay_OnlineStatus
   widget=Image_ping
   properties {brush {
     bIsDynamicallyLoaded=false
     drawAs=Image
     imageType=NoImage
     mirroring=NoMirror
     resourceName=None
     resourceObject=/Game/UI/Events/Leaderboards/Materials/M_UI_LiveIcon_Inst
     tiling=NoTile
     imageSize=<60 60>
     margin=<ltrb 0 0 0 0>
     outlineSettings {
      bUseBrushTransparency=false
      roundingType=HalfHeightRadius
      width=0
      color {colorUseRule=UseColor_Specified, specifiedColor=<rgba 0 0 0 0>}
      cornerRadii=<0 0 0 0>
     }
     tintColor {colorUseRule=UseColor_Specified, specifiedColor=<rgba 1 1 1 1>}
     uVRegion {bIsValid=false, max=<0 0>, min=<0 0>}
    }}
   slot {class=OverlaySlot, properties {padding=<ltrb 0 -10 0 -10>}}
  }
  Image_OnlineStatus {
   widgetClassPath=&Image
   bIsVariable=true
   parent=Overlay_OnlineStatus
   widget=Image_OnlineStatus
   properties {
    visibility=SelfHitTestInvisible
    brush {
     bIsDynamicallyLoaded=false
     drawAs=Image
     imageType=NoImage
     mirroring=NoMirror
     resourceName=None
     resourceObject=/Game/UI/Events/Leaderboards/Materials/M_UI_LiveIconCenter_Inst
     tiling=NoTile
     imageSize=<30 30>
     margin=<ltrb 0 0 0 0>
     outlineSettings {
      bUseBrushTransparency=false
      roundingType=HalfHeightRadius
      width=0
      color {colorUseRule=UseColor_Specified, specifiedColor=<rgba 0 0 0 0>}
      cornerRadii=<0 0 0 0>
     }
     tintColor {colorUseRule=UseColor_Specified, specifiedColor=<rgba 1 1 1 1>}
     uVRegion {bIsValid=false, max=<0 0>, min=<0 0>}
    }
    colorAndOpacity=<rgba 1 0 0 1>
   }
   slot {class=OverlaySlot, properties {horizontalAlignment=HAlign_Center, verticalAlignment=VAlign_Center}}
  }
  Text_Score {
   widgetClassPath=&CommonTextBlock
   bIsVariable=true
   parent=HorizontalBox_1
   widget=Text_Score
   properties {
    justification=Center
    text=99
    font {
     bForceMonospaced=false
     bMaterialIsStencil=false
     fontMaterial=null
     fontObject=&Burbank
     letterSpacing=0
     monospacedWidth=1
     size=24
     skewAmount=0
     typefaceFontName=Burbank-PlayerPower
     outlineSettings {
      bApplyOutlineToDropShadows=false
      bMiteredCorners=false
      bSeparateFillAlpha=false
      outlineMaterial=null
      outlineSize=0
      outlineColor=<rgba 0 0 0 1>
     }
    }
    shadowOffset=<0 0>
   }
   slot {class=HorizontalBoxSlot, properties {horizontalAlignment=HAlign_Right, verticalAlignment=VAlign_Center, padding=<ltrb 34 0 34 0>}}
  }
  ScaleBox_1 {widgetClassPath=&ScaleBox, parent=HorizontalBox_2, widget=ScaleBox_1, slot {class=HorizontalBoxSlot}}
  SizeBox_2 {
   widgetClassPath=&SizeBox
   parent=ScaleBox_1
   widget=SizeBox_2
   properties {bOverride_HeightOverride=true, bOverride_WidthOverride=true, heightOverride=40, widthOverride=40}
   slot {class=ScaleBoxSlot}
  }
  FlagImage {
   widgetClassPath=/Script/FortniteUI.FortFlagImage
   bInherited=true
   bIsVariable=true
   parent=SizeBox_2
   widget=FlagImage
   properties {
    materialTextureParamName=Texture
    brush {
     bIsDynamicallyLoaded=false
     drawAs=Image
     imageType=NoImage
     mirroring=NoMirror
     resourceName=None
     resourceObject=/Game/UI/Events/Leaderboards/Materials/M_UI_FlagMaterial_Inst
     tiling=NoTile
     imageSize=<128 128>
     margin=<ltrb 0 0 0 0>
     outlineSettings {
      bUseBrushTransparency=false
      roundingType=HalfHeightRadius
      width=0
      color {colorUseRule=UseColor_Specified, specifiedColor=<rgba 0 0 0 0>}
      cornerRadii=<0 0 0 0>
     }
     tintColor {colorUseRule=UseColor_Specified, specifiedColor=<rgba 1 1 1 1>}
     uVRegion {bIsValid=false, max=<0 0>, min=<0 0>}
    }
   }
   slot {class=SizeBoxSlot}
  }
  InputActionWidget {widgetClassPath=&CommonActionWidget, bInherited=true, widget=null}
 ]
}
