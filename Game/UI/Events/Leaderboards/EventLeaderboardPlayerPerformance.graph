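# Event leaderboard widget, derived from FortShowdownDetailView, that shows the player's best event score and "Top X%" standing in one row.
# Refresh Data fills both texts from inherited getters, leaving them empty when the event has no score.
# The percentile is scaled to a whole percentage, rounded away from zero, and collapses unless above zero.
# The whole row collapses when the best score is negative, and a down-only ScaleBox shrinks it to fit.
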WidgetBlueprint /Game/UI/Events/Leaderboards/EventLeaderboardPlayerPerformance
graphs [EventGraph {nodes [
   Event Refresh Data() {is=title K2Node_Event}
   n1 = GetBestEventScore!(self:Fort Showdown Detail View Object Reference) -> (ReturnValue:int=0)
   > "SetText (Text)"(
    self=Get Text_PlayerScore() -> (Text_PlayerScore:Common Text Object Reference)
    InText=Select(
     Option 0:Text=GetEmptyText!() -> (ReturnValue:Text)
     Option 1:Text=Format Text!(Format="LOCTEXT(\"{Score}\")", Score:int=@n1) -> (Result)
     Index=HasScoreInEvent!(self:Fort Showdown Detail View Object Reference) -> (ReturnValue:bool)
    ) -> (ReturnValue:Text)
   )
   n2 = "To Text (Float)"(
    Value=float * float(A=GetEventLeaderboardPercentile!(self:Fort Showdown Detail View Object Reference) -> (ReturnValue:float), B=100)
    RoundingMode=FromZero
    MinimumIntegralDigits=1
    MaximumIntegralDigits=324
    MinimumFractionalDigits=0
    MaximumFractionalDigits=0
   )
   > "SetText (Text)"(
    self=Get Text_PlayerPerformance() -> (Text_PlayerPerformance:Common Text Object Reference)
    InText=Select(
     Option 0=GetEmptyText()
     Option 1=Format Text!(Format="LOCTEXT(\"(Top {Percentile}%)\")", Percentile:Text=@n2) -> (Result)
     Index=HasScoreInEvent()
    )
   )
   n3 = GetEventLeaderboardPercentile()
   > SetVisibility(
    self=Get Text_PlayerPerformance()
    InVisibility=Select(
     Option 0:ESlateVisibility Enum=Collapsed
     Option 1:ESlateVisibility Enum=SelfHitTestInvisible
     Index="float > float"(A=@n3, B=0) {is=title K2Node_CallFunction}
    ) -> (ReturnValue:ESlateVisibility Enum)
   )
   n4 = GetBestEventScore() -> (ReturnValue=0)
   > SetVisibility(
    self=Get HorizontalBox_Container() -> (HorizontalBox_Container:Horizontal Box Object Reference)
    InVisibility=Select(
     Option 0=HitTestInvisible
     Option 1=Collapsed
     Index="integer < integer"!(A:int=@n4, B:int=0) -> (ReturnValue:bool) {is=title K2Node_CallFunction}
    )
   )
  ]}]
variables []
widgets {
 info {inheritedWidgetCount=6, parentClass=/Script/FortniteUI.FortShowdownDetailView, rootWidgetClass=&ScaleBox, widgetCount=12}
 widgets [
  ScaleBox_0 {widgetClassPath=&ScaleBox, widget=ScaleBox_0, properties {stretchDirection=DownOnly}}
  HorizontalBox_Container {
   widgetClassPath=&HorizontalBox
   bIsVariable=true
   parent=ScaleBox_0
   widget=HorizontalBox_Container
   slot {class=ScaleBoxSlot, properties {horizontalAlignment=HAlign_Fill, verticalAlignment=VAlign_Fill}}
  }
  CommonTextBlock_0 {
   widgetClassPath=&CommonTextBlock
   bIsVariable=true
   parent=HorizontalBox_Container
   widget=CommonTextBlock_0
   properties {
    text=Your Score:
    font {
     bForceMonospaced=false
     bMaterialIsStencil=false
     fontMaterial=null
     fontObject=&Burbank
     letterSpacing=0
     monospacedWidth=1
     size=24
     skewAmount=0
     typefaceFontName=Burbank-PlayerPower
     outlineSettings {
      bApplyOutlineToDropShadows=false
      bMiteredCorners=false
      bSeparateFillAlpha=false
      outlineMaterial=null
      outlineSize=0
      outlineColor=<rgba 0 0 0 1>
     }
    }
   }
   slot {class=HorizontalBoxSlot, properties {verticalAlignment=VAlign_Center}}
  }
  Text_PlayerScore {
   widgetClassPath=&CommonTextBlock
   bIsVariable=true
   parent=HorizontalBox_Container
   widget=Text_PlayerScore
   properties {
    text=99
    font {
     bForceMonospaced=false
     bMaterialIsStencil=false
     fontMaterial=null
     fontObject=&Burbank
     letterSpacing=0
     monospacedWidth=1
     size=37.5
     skewAmount=0
     typefaceFontName=Burbank-PlayerPower
     outlineSettings {
      bApplyOutlineToDropShadows=false
      bMiteredCorners=false
      bSeparateFillAlpha=false
      outlineMaterial=null
      outlineSize=0
      outlineColor=<rgba 0 0 0 1>
     }
    }
   }
   slot {class=HorizontalBoxSlot, properties {verticalAlignment=VAlign_Center, padding=<ltrb 15 -10 0 -10>}}
  }
  CommonTextBlock_1 {
   widgetClassPath=&CommonTextBlock
   bIsVariable=true
   parent=HorizontalBox_Container
   widget=CommonTextBlock_1
   properties {
    text=PTS
    font {
     bForceMonospaced=false
     bMaterialIsStencil=false
     fontMaterial=null
     fontObject=&Burbank
     letterSpacing=0
     monospacedWidth=1
     size=24
     skewAmount=0
     typefaceFontName=Burbank-PlayerPower
     outlineSettings {
      bApplyOutlineToDropShadows=false
      bMiteredCorners=false
      bSeparateFillAlpha=false
      outlineMaterial=null
      outlineSize=0
      outlineColor=<rgba 0 0 0 1>
     }
    }
   }
   slot {class=HorizontalBoxSlot, properties {horizontalAlignment=HAlign_Left, verticalAlignment=VAlign_Center, padding=<ltrb 10 0 15 0>}}
  }
  Text_PlayerPerformance {
   widgetClassPath=&CommonTextBlock
   bIsVariable=true
   parent=HorizontalBox_Container
   widget=Text_PlayerPerformance
   properties {
    text=(Top 98%)
    font {
     bForceMonospaced=false
     bMaterialIsStencil=false
     fontMaterial=null
     fontObject=&Burbank
     letterSpacing=0
     monospacedWidth=1
     size=24
     skewAmount=0
     typefaceFontName=Burbank-PlayerPower
     outlineSettings {
      bApplyOutlineToDropShadows=false
      bMiteredCorners=false
      bSeparateFillAlpha=false
      outlineMaterial=null
      outlineSize=0
      outlineColor=<rgba 0 0 0 1>
     }
    }
   }
   slot {class=HorizontalBoxSlot, properties {verticalAlignment=VAlign_Center}}
  }
  Text_TimeRemaining {widgetClassPath=&CommonTextBlock, bInherited=true, widget=null}
  Text_DebugId {widgetClassPath=&TextBlock, bInherited=true, widget=null, properties=~}
  Image_PosterFront {widgetClassPath=/Script/EpicCMSUIFramework.EpicCMSImage, bInherited=true, widget=null}
  Image_PosterBack {widgetClassPath=&EpicCMSImage, bInherited=true, widget=null}
  Image_LoadingScreen {widgetClassPath=&EpicCMSImage, bInherited=true, widget=null}
  Image_PlaylistTile {widgetClassPath=&EpicCMSImage, bInherited=true, widget=null}
 ]
}
