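# Widget for a numbered stat icon in event team match history, layering a centered text block over an alt-stat image.
# Set Stat Text fills the text and applies StatColor, black by default.
# Hover plays the Selected animation forward, fading text from black to white and the icon the opposite way.
# Unhover reverses it, then restores StatColor on finish if bSetTextColor is still true, a flag that hovering clears.
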WidgetBlueprint /Game/UI/Events/Leaderboards/EventTeamMatchHistoryStatNumberedIcon
animations [Selected {
  binding_count=2
  length_seconds=0.25
  object=&self:Selected
  sequence {bindings [
    Text_StatType {
     id=#1
     possessed_class=CommonTextBlock
     tracks [{
       type=MovieSceneColorTrack
       display_name=Color and Opacity
       sections [{
         type=MovieSceneColorSection
         end_frame=5
         start_frame=0
         channels [
          Color.R {type=MovieSceneScriptingFloatChannel, keys [[0,0],[5,1]]}
          Color.G {type=MovieSceneScriptingFloatChannel, keys [[0,0],[5,1]]}
          Color.B {type=MovieSceneScriptingFloatChannel, keys [[0,0],[5,1]]}
          Color.A {type=MovieSceneScriptingFloatChannel, const [1,2]}
         ]
        }]
      }]
    }
    Image_AlternativeStat {
     id=#2
     possessed_class=Image
     tracks [{
       type=MovieSceneColorTrack
       display_name=Color and Opacity
       sections [{
         type=MovieSceneColorSection
         end_frame=5
         start_frame=0
         channels [
          Color.R {type=MovieSceneScriptingFloatChannel, thinned_from=3, keys [[0,1],[5,0]]}
          Color.G {type=MovieSceneScriptingFloatChannel, thinned_from=3, keys [[0,1],[5,0]]}
          Color.B {type=MovieSceneScriptingFloatChannel, thinned_from=3, keys [[0,1],[5,0]]}
          Color.A {type=MovieSceneScriptingFloatChannel, const [1,3]}
         ]
        }]
      }]
    }
   ]}
 }]
graphs [EventGraph {nodes [
   n1 = Set Stat Text!() -> (OutputDelegate, then, NewText:Text) {is=title K2Node_CustomEvent}
   n2 = Get Text_StatType() -> (Text_StatType:Common Text Object Reference)
   > "SetText (Text)"(self=@n2, InText=@n1.NewText)
   > SetColorAndOpacity(self=@n2, InColorAndOpacity=Get StatColor() -> (StatColor:Slate Color Structure))
   Set Hovered Event() {is=title K2Node_CustomEvent}
   n3 = Get Selected() -> (Selected:Widget Animation Object Reference)
   > PlayAnimationForward(InAnimation=@n3)
   > Set bSetTextColor(bSetTextColor:bool=false) -> (Output_Get:bool)
   Set Unhovered Event() {is=title K2Node_CustomEvent}
   > n4 = "Play Animation with Finished event (legacy)"!(
    execute:Exec
    Widget:User Widget Object Reference=Self-Reference()
    InAnimation:Widget Animation Object Reference=@n3
    StartAtTime:float=0
    NumLoopsToPlay:int=1
    PlayMode:EUMGSequencePlayMode Enum=Reverse
    PlaybackSpeed:float=1
   ) -> (then:Exec, Result:UMGSequence Player Object Reference, Finished:Exec) {is=title K2Node_PlayAnimation}
   > Set bSetTextColor(bSetTextColor=true)
   Branch(execute=@n4.Finished, Condition=Get bSetTextColor() -> (bSetTextColor:bool))
   > SetColorAndOpacity(self=Get Text_StatType(), InColorAndOpacity=Get StatColor())
  ]}]
variables [
 bSetTextColor {type=boolean}
 StatColor {type=SlateColor, default {colorUseRule=UseColor_Specified, specifiedColor=<rgba 0 0 0 1>}}
]
widgets {
 info {inheritedWidgetCount=0, parentClass=&UserWidget, rootWidgetClass=&Overlay, widgetCount=3}
 widgets [
  Overlay_3 {widgetClassPath=&Overlay, widget=Overlay_3, properties {bIsEnabled=false}}
  Image_AlternativeStat {
   widgetClassPath=&Image
   parent=Overlay_3
   widget=Image_AlternativeStat
   properties {
    bIsEnabled=false
    visibility=SelfHitTestInvisible
    brush {
     bIsDynamicallyLoaded=false
     drawAs=Image
     imageType=NoImage
     mirroring=NoMirror
     resourceName=None
     resourceObject=/Game/UI/Foundation/Textures/Icons/Athena/T_T_Icon_Athena_AltStat
     tiling=NoTile
     imageSize=<32 32>
     margin=<ltrb 0 0 0 0>
     outlineSettings {
      bUseBrushTransparency=false
      roundingType=HalfHeightRadius
      width=0
      color {colorUseRule=UseColor_Specified, specifiedColor=<rgba 0 0 0 0>}
      cornerRadii=<0 0 0 0>
     }
     tintColor {colorUseRule=UseColor_Specified, specifiedColor=<rgba 1 1 1 1>}
     uVRegion {bIsValid=false, max=<0 0>, min=<0 0>}
    }
   }
   slot {class=OverlaySlot, properties {horizontalAlignment=HAlign_Center, verticalAlignment=VAlign_Center}}
  }
  Text_StatType {
   widgetClassPath=&CommonTextBlock
   bIsVariable=true
   parent=Overlay_3
   widget=Text_StatType
   properties {
    bIsEnabled=false
    bIsScrollingEnabled=false
    justification=Center
    text=0
    colorAndOpacity {colorUseRule=UseColor_Specified, specifiedColor=<rgba 0 0 0 1>}
    font {
     bForceMonospaced=false
     bMaterialIsStencil=false
     fontMaterial=null
     fontObject=&Burbank
     letterSpacing=0
     monospacedWidth=1
     size=13
     skewAmount=0.06
     typefaceFontName=Burbank-Black
     outlineSettings {
      bApplyOutlineToDropShadows=false
      bMiteredCorners=false
      bSeparateFillAlpha=false
      outlineMaterial=null
      outlineSize=0
      outlineColor=<rgba 0 0 0 1>
     }
    }
    shadowOffset=<0 0>
   }
   slot {class=OverlaySlot, properties {horizontalAlignment=HAlign_Center, verticalAlignment=VAlign_Center, padding=<ltrb 0 1.5 1 0>}}
  }
 ]
}
