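# Event leaderboard list widget, built on FortLeaderboardListViewWrapper, showing a scrolling entry list, a throbber loading row and a player performance bar below.
# Set Display Info stores tournament display data and tints the bar with its right background color, capping HSV brightness to limit glare.
# Set Event passes event and window IDs to the performance widget with a forced refresh.
# Leaderboard display type changes show or collapse that widget.
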WidgetBlueprint /Game/UI/Events/Leaderboards/LeaderboardListViewWrapper
graphs [EventGraph {nodes [
   n1 = Set Display Info!() -> (OutputDelegate, then, DisplayInfo:Fort Tournament Display Info Structure) {is=title K2Node_CustomEvent}
   > n2 = Set DisplayInfo(DisplayInfo:Fort Tournament Display Info Structure=@n1.DisplayInfo) -> (Output_Get:Fort Tournament Display Info Structure)
   n3 = RGB to HSV!(InColor:Linear Color Structure=Break Fort Tournament Display Info!(FortTournamentDisplayInfo:"Fort Tournament Display Info Structure (by ref)"=@n2) -> (
     TitleLine1:Text
     TitleLine2:Text
     ScheduleInfo:Text
     FlavorDescription:Text
     DetailsDescription:Text
     ShortFormatTitle:Text
     LongFormatTitle:Text
     BackgroundTitle:Text
     PinScoreRequirement:int
     PinEarnedText:Text
     BaseColor:Linear Color Structure
     PrimaryColor:Linear Color Structure
     SecondaryColor:Linear Color Structure
     HighlightColor:Linear Color Structure
     TitleColor:Linear Color Structure
     ShadowColor:Linear Color Structure
     BackgroundLeftColor:Linear Color Structure
     BackgroundRightColor:Linear Color Structure
     BackgroundTextColor:Linear Color Structure
     PosterFadeColor:Linear Color Structure
     AlertText:Text
     SeriesPointLeaderboardName:Text
     AlertType:EFortTournamentAlertType Enum
     RoundDisplayNames:Array of Texts
    ) {is=title K2Node_BreakStruct}.BackgroundRightColor) -> (H:float, S:float, V:float, A:float)
   > SetBrushColor(
    self=Get CommonBorder_OptionsBarBG() -> (CommonBorder_OptionsBarBG:Common Border Object Reference)
    InBrushColor=HSV to RGB!(H:float=@n3, S:float=@n3.S, V:float="Clamp (Float)"(Value=@n3.V, Min=0, Max=0.7), A:float=@n3.A) -> (ReturnValue:Linear Color Structure)
   )
   n4 = Set Event!() -> (OutputDelegate, then, EventId:String, EventWindowId:String) {is=title K2Node_CustomEvent}
   > SetTournament!(
    execute
    self:Fort Showdown Detail View Object Reference=Get EventLeaderboardPlayerPerformance() -> (EventLeaderboardPlayerPerformance:Event Leaderboard Player Performance Object Reference)
    TournamentId:String=@n4.EventId
    EventId:String=@n4.EventWindowId
    bForceRefreshUI:bool=true
   ) -> (then)
   n5 = Event OnLeaderboardDisplayTypeChanged!() -> (OutputDelegate, then, NewDisplayType:ELeaderboardDisplayType Enum=Default) {is=title K2Node_Event}
   > SetVisibility(
    self=Get EventLeaderboardPlayerPerformance()
    InVisibility=Select!(
     Default:ESlateVisibility Enum=SelfHitTestInvisible
     Extra:ESlateVisibility Enum=Collapsed
     Index:ELeaderboardDisplayType Enum=@n5.NewDisplayType
    ) -> (ReturnValue)
   )
  ]}]
variables [DisplayInfo {
  type=FortTournamentDisplayInfo
  default {
   alertText=""
   alertType=Warning
   backgroundTitle=""
   detailsDescription=""
   flavorDescription=""
   longFormatTitle=""
   pinEarnedText=""
   pinScoreRequirement=0
   scheduleInfo=""
   seriesPointLeaderboardName=""
   shortFormatTitle=""
   titleLine1=""
   titleLine2=""
   backgroundLeftColor=<rgba 0 0 0 0>
   backgroundRightColor=<rgba 0 0 0 0>
   backgroundTextColor=<rgba 0 0 0 0>
   baseColor=<rgba 0 0 0 0>
   highlightColor=<rgba 0 0 0 0>
   posterFadeColor=<rgba 0 0 0 0>
   primaryColor=<rgba 0 0 0 0>
   roundDisplayNames []
   secondaryColor=<rgba 0 0 0 0>
   shadowColor=<rgba 0 0 0 0>
   titleColor=<rgba 0 0 0 0>
  }
 }]
widgets {
 info {
  inheritedWidgetCount=2
  parentClass=/Script/FortniteUI.FortLeaderboardListViewWrapper
  rootWidgetClass=&VerticalBox
  widgetCount=9
 }
 widgets [
  VerticalBox {widgetClassPath=&VerticalBox, bIsVariable=true, widget=VerticalBox}
  ListView_LeaderboardEntries {
   widgetClassPath=&CommonListView
   bInherited=true
   bIsVariable=true
   parent=VerticalBox
   widget=ListView_LeaderboardEntries
   properties {
    clipping=ClipToBoundsWithoutIntersecting
    entryWidgetClass=&/EventLeaderboardEntry.EventLeaderboardEntry_C
    numDesignerPreviewEntries=20
   }
   slot {class=VerticalBoxSlot, properties {size {sizeRule=Fill, value=1}}}
  }
  Content_LoadingEntries {
   widgetClassPath=&HorizontalBox
   bInherited=true
   bIsVariable=true
   parent=VerticalBox
   widget=Content_LoadingEntries
   properties {visibility=Collapsed}
   slot {class=VerticalBoxSlot, properties {horizontalAlignment=HAlign_Center, padding=<ltrb 20 0 20 20>, size {sizeRule=Fill, value=1}}}
  }
  ScaleBox_0 {
   widgetClassPath=&ScaleBox
   parent=Content_LoadingEntries
   widget=ScaleBox_0
   properties {stretchDirection=DownOnly}
   slot {class=HorizontalBoxSlot, properties {verticalAlignment=VAlign_Top, padding=<ltrb 25 0 25 0>}}
  }
  CommonTextBlock_0 {
   widgetClassPath=&CommonTextBlock
   parent=ScaleBox_0
   widget=CommonTextBlock_0
   properties {
    text=Loading Leaderboard Entries
    textTransformPolicy=ToUpper
    font {
     bForceMonospaced=false
     bMaterialIsStencil=false
     fontMaterial=null
     fontObject=&BurbankSmall
     letterSpacing=0
     monospacedWidth=1
     size=18
     skewAmount=0
     typefaceFontName=Bold
     outlineSettings {
      bApplyOutlineToDropShadows=false
      bMiteredCorners=false
      bSeparateFillAlpha=false
      outlineMaterial=null
      outlineSize=0
      outlineColor=<rgba 0 0 0 1>
     }
    }
    shadowOffset=<0 0>
   }
   slot {class=ScaleBoxSlot}
  }
  Image_0 {
   widgetClassPath=&Image
   bIsVariable=true
   parent=Content_LoadingEntries
   widget=Image_0
   properties {
    visibility=SelfHitTestInvisible
    brush {
     bIsDynamicallyLoaded=false
     drawAs=Image
     imageType=NoImage
     mirroring=NoMirror
     resourceName=None
     resourceObject=/FortUILibrary/UIKit/Throbber/MI_UIKit_Monochrome_Throbber
     tiling=NoTile
     imageSize=<32 32>
     margin=<ltrb 0 0 0 0>
     outlineSettings {
      bUseBrushTransparency=false
      roundingType=HalfHeightRadius
      width=0
      color {colorUseRule=UseColor_Specified, specifiedColor=<rgba 0 0 0 0>}
      cornerRadii=<0 0 0 0>
     }
     tintColor {colorUseRule=UseColor_Specified, specifiedColor=<rgba 1 1 1 1>}
     uVRegion {bIsValid=false, max=<0 0>, min=<0 0>}
    }
   }
   slot {class=HorizontalBoxSlot, properties {horizontalAlignment=HAlign_Center, verticalAlignment=VAlign_Top}}
  }
  CommonBorder_OptionsBarBG {
   widgetClassPath=&CommonBorder
   bIsVariable=true
   parent=VerticalBox
   widget=CommonBorder_OptionsBarBG
   properties {
    background {
     bIsDynamicallyLoaded=false
     drawAs=Image
     imageType=NoImage
     mirroring=NoMirror
     resourceName=None
     resourceObject=null
     tiling=NoTile
     imageSize=<32 32>
     margin=<ltrb 0 0 0 0>
     outlineSettings {
      bUseBrushTransparency=false
      roundingType=HalfHeightRadius
      width=0
      color {colorUseRule=UseColor_Specified, specifiedColor=<rgba 0 0 0 0>}
      cornerRadii=<0 0 0 0>
     }
     tintColor {colorUseRule=UseColor_Specified, specifiedColor=<rgba 1 1 1 1>}
     uVRegion {bIsValid=false, max=<0 0>, min=<0 0>}
    }
    brushColor=<rgba 0.006 0.0742 0.4969 1>
    padding=<ltrb 0 0 0 0>
   }
   slot {class=VerticalBoxSlot}
  }
  Box_FliterOptions {
   widgetClassPath=&HorizontalBox
   parent=CommonBorder_OptionsBarBG
   widget=Box_FliterOptions
   slot {class=BorderSlot, properties {padding=<ltrb 0 0 0 0>}}
  }
  EventLeaderboardPlayerPerformance {
   widgetClassPath=&/EventLeaderboardPlayerPerformance
   bIsVariable=true
   parent=Box_FliterOptions
   widget=EventLeaderboardPlayerPerformance
   slot {class=HorizontalBoxSlot, properties {horizontalAlignment=HAlign_Center, padding=<ltrb 20 6 20 6>, size {sizeRule=Fill, value=1}}}
  }
 ]
}
